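; Gameplay ability that sends a ThawEpsonAI NPC pawn back to its leash location.
; On activation it commits, casts the avatar to the NPC pawn and starts a MoveAI task to that location, using pathfinding and the Phoebe no-smashable nav filter.
; The ability ends when the move finishes or if the cast or task creation fails.
; On end it unbinds the move-finished delegate and clears the stored task.
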(event Ability|EventActivateAbility
  (Ability|CommitAbility)
  (bind _asnpc_pawn_thaw_epson_ai (Utilities|Casting|CastToNPC_Pawn_ThawEpsonAI (Ability|GetAvatarActorfromActorInfo))
    (:then
      (bind _asynctaskproxy (Ability|Tasks|MoveAI "MoveToEQS" 0 (Variables|/ThawEpsonAI/NPCs/Pawns/NPC_Pawn_ThawEpsonAI.NPC_Pawn_ThawEpsonAI_C|GetLeashLocation _asnpc_pawn_thaw_epson_ai) "UsePathfinding" "Ignore" true "Default" false true 4e+01 0 "/Game/Athena/AI/Phoebe/NavFilter/NavFilter_Phoebe_NoSmashable.NavFilter_Phoebe_NoSmashable_C" "(TagName=\"\")"))
      (Utilities|IsValid _asynctaskproxy
        (:Is%20Valid
          (bind _output_get (Variables|MovetoEQS|SetMoveTask _asynctaskproxy))
          (EventDispatchers|BindEventtoOnMoveFinished _output_get (delegate Custom|OnMoveFinishedEvent)))
        (:Is%20Not%20Valid
          (Ability|EndAbility))))
    (:CastFailed
      (Ability|EndAbility))))

(event Custom|OnMoveFinishedEvent
  (params (Result :type EPathFollowingResult%20Enum :default "Success") (AIController :type AIController%20Object%20Reference))
  (Ability|EndAbility))

(event Ability|EventOnEndAbility
  (params (bWasCancelled :type Boolean))
  (bind _movetask (Variables|MovetoEQS|GetMoveTask)
    (:then
      (EventDispatchers|UnbindallEventsfromOnMoveFinished _movetask)
      (Variables|MovetoEQS|SetMoveTask 0))
    (:else)))
