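; Gameplay ability that manages the frightened state of the ThawEpsonAI NPC.
; On activation it watches pawn health, applying the frightened effect and a ColorID-parameterized cue once the health ratio reaches a scalable threshold.
; A ForceFrightened gameplay event applies the effect directly, without the cue.
; The pawn's OnReturnedtoNormal dispatcher clears the flag and removes the effect via its stored handle, re-arming the health check.
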(event Ability|EventActivateAbility
  (Variables|InternalData|SetIsFrightened false)
  (bind _asfort_pawn (Utilities|Casting|CastToFortPawn (Ability|GetAvatarActorfromActorInfo))
    (:then
      (bind _output_get (Variables|InternalData|SetAIPawn _asfort_pawn))
      (Health|BindEventtoOnPawnHealthChanged _output_get (EventDispatchers|CreateEvent :!event Custom|PawnHealthChanged))
      (bind _asnpc_pawn_thaw_epson_ai (Utilities|Casting|CastToNPC_Pawn_ThawEpsonAI _output_get))
      (Default|BindEventtoOnReturnedtoNormal _asnpc_pawn_thaw_epson_ai (EventDispatchers|CreateEvent :!event Custom|NoLongerFrightened))
      (Ability|Tasks|WaitGameplayEvent "(TagName=\"Event.CosmicThunder.ThawEpsonAI.ForceFrightened\")"
        (:EventReceived
          (CallFunction|ForceFrightened))))
    (:CastFailed
      (Ability|EndAbility))))

(event Ability|EventOnEndAbility
  (params (bWasCancelled :type Boolean)))

(event Custom|ForceFrightened
  (Variables|InternalData|SetIsFrightened true)
  (bind _returnvalue (Ability|ApplyGameplayEffectToOwner (Variables|Default|GetGE_IsFrightened)))
  (Variables|InternalData|SetHNDL_GEIsFrightened _returnvalue))

(event Custom|PawnHealthChanged
  (if (Variables|InternalData|GetIsFrightened)
    _
    (elif (<= (Math|Float|SafeDivide (FortPawn|GetHealth (Variables|InternalData|GetAIPawn)) (FortPawn|GetMaxHealth (Variables|InternalData|GetAIPawn))) (ScalableFloat|GetValueatLevel (Variables|Default|GetFrightenedThreshold)))
      (bind _asnpc_pawn_thaw_epson_ai (Utilities|Casting|CastToNPC_Pawn_ThawEpsonAI (Variables|InternalData|GetAIPawn)))
      (Ability|ExecuteGameplayCueWithParamsOnOwner (Variables|Default|GetGC_TurnedFrightened) (Ability|GameplayCue|MakeGameplayCueParameters 0.0 (Variables|/ThawEpsonAI/NPCs/Pawns/NPC_Pawn_ThawEpsonAI.NPC_Pawn_ThawEpsonAI_C|GetColorID _asnpc_pawn_thaw_epson_ai)))
      (Variables|InternalData|SetIsFrightened true)
      (bind _returnvalue (Ability|ApplyGameplayEffectToOwner (Variables|Default|GetGE_IsFrightened)))
      (Variables|InternalData|SetHNDL_GEIsFrightened _returnvalue))))

(event Custom|NoLongerFrightened
  (Variables|InternalData|SetIsFrightened false)
  (Ability|RemoveGameplayEffectFromOwnerWithHandle (Variables|InternalData|GetHNDL_GEIsFrightened)))
