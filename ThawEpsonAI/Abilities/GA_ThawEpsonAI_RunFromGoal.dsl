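; Gameplay ability for the ThawEpson AI NPC that extends a parent move ability, named for running from a goal.
; It starts the move only if the first location found lies within the NPC's leash radius, otherwise ending the ability.
; Finished moves idle for a random scalable delay before the parent's completion logic.
; Blocked moves trigger target selection to attack surroundings with effects, an event and a cue, then restart the move.
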(event Ability|EventActivateAbility
  (|CallParentFunction) ; This node is disabled and will not be called.\nDrag off pins to build functionality.
)

(event Ability|EventOnEndAbility
  (params (bWasCancelled :type Boolean))
  (|CallParentFunction bWasCancelled) ; This node is disabled and will not be called.\nDrag off pins to build functionality.
)

(event EventStartMoveAI
  ;[1 0.9996 0.2299] Validate Leash
  (bind _asnpc_pawn_thaw_epson_ai (Utilities|Casting|CastToNPC_Pawn_ThawEpsonAI (Ability|GetAvatarActorfromActorInfo)))
  (bind _isinleashradius (Class|NPCPawnThawEpsonAI|IsInsideLeashRadius _asnpc_pawn_thaw_epson_ai 0 (Variables|InternalData|GetFirstLocationFound)))
  (if _isinleashradius
    (|CallParentFunction)
    (else
      (Ability|EndAbility))))

(event Custom|StartIdleAndEndMoveTo
  (Utilities|FlowControl|Delay (Math|Random|RandomFloatinRange (ScalableFloat|GetValueatLevel (Variables|Default|GetMin_PauseDelayAfterMove)) (ScalableFloat|GetValueatLevel (Variables|Default|GetMax_PauseDelayAfterMove)))) ; If no plants found, stay here for a while...
  (|CallParentFunction (stale EventOnMoveFinishedEvent Result) (stale EventOnMoveFinishedEvent AIController)))

(event EventOnMoveFinishedEvent
  (params (Result :type EPathFollowingResult%20Enum :default "Success") (AIController :type AIController%20Object%20Reference))
  (switch Utilities|FlowControl|Switch|SwitchonEPathFollowingResult Result
    (:Success
      _)
    (:Blocked
      (CallFunction|TryDamageSurroundings))
    (:OffPath
      _)
    (:Aborted
      _)
    (:Invalid
      _))
  (CallFunction|StartIdleandEndMoveTo))

(event Custom|TryDamageSurroundings
  ;[1 1 1] If Blocked Try to Damage ourselves out
  (bind (_asynctaskproxy _targetdata _applicationtag) (Ability|Tasks|DoTargetSelection "None" "(TagName=\"Targeting.ApplicationTag.0\")")
    (:Targeted
      (if (> (Utilities|Array|Length (Ability|TargetData|GetAllActorsfromTargetData _targetdata)) 0)
        (Ability|ApplyGameplayEffectContainer _targetdata _applicationtag 0)
        (Ability|SendGameplayEvent (Variables|Default|GetTagToSendWhenAttackingWhileBlocked))
        (Ability|ExecuteGameplayCueOnOwner (Variables|Default|GetTag_AttackingCue))
        (CallFunction|StartMoveAI)
        (else
          (|CallParentFunction (stale EventOnMoveFinishedEvent Result) (stale EventOnMoveFinishedEvent AIController)))))))
